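# Conversation database for the Clyde LemonCart boss NPC, a small question-and-answer dialogue menu.
# The Clyde Boss Agency entry tag opens a greeting speech node that branches to questions about the location and loot, plus a Back option.
# Each question plays a short spoken reply, then offers a ReturnToStart choice to go back to the start.
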ConversationDatabase /Clyde/Athena/AI/Clyde/Bosses/LemonCart/NPCLibrary_Conversation_Clyde_LemonCartBoss
entry_points [{tag=Conversation.Entry.Clyde.Boss.Agency, destinations [#1]}]
nodes [
 ConversationEntryPointNode_0 {
  id=#1
  class=ConversationEntryPointNode
  outputs [#2]
  properties {entryTag {tagName=Conversation.Entry.Clyde.Boss.Agency}}
 }
 Speech_C_0 {
  id=#2
  class=Speech_C
  outputs [#3, #4, #5]
  properties {nodeName="", softTaskIcon=&T_UI_NPC_question, generalConfig {defaultMessage=What's poppin'?}}
  sub_nodes [ConversationChoiceNode_5 {class=ConversationChoiceNode, properties {defaultChoiceDisplayText=Ask about...}}]
 }
 Speech_C_4 {
  id=#5
  class=Speech_C
  outputs [#6]
  properties {nodeName="", softTaskIcon=&T_UI_NPC_question, generalConfig {defaultMessage="Leave the Snooping around to me. "}}
  sub_nodes [ConversationChoiceNode_3 {class=ConversationChoiceNode, properties {defaultChoiceDisplayText=Got any loot around here?}}]
 }
 ReturnToStart_C_0 {
  id=#6
  class=ReturnToStart_C
  properties {softTaskIcon=&T_UI_NPC_Back}
  sub_nodes [ConversationChoiceNode_1 {class=ConversationChoiceNode, properties {defaultChoiceDisplayText=BACK}}]
 }
 Back_C_3 {
  id=#4
  class=Back_C
  sub_nodes [ConversationChoiceNode_23 {class=ConversationChoiceNode, properties {defaultChoiceDisplayText=BACK}}]
 }
 Speech_C_3 {
  id=#3
  class=Speech_C
  outputs [#7]
  properties {nodeName="", softTaskIcon=&T_UI_NPC_question, generalConfig {defaultMessage=Welcome to the Doggpound.}}
  sub_nodes [ConversationChoiceNode_0 {class=ConversationChoiceNode, properties {defaultChoiceDisplayText=Where am I?}}]
 }
 ReturnToStart_C_1 {
  id=#7
  class=ReturnToStart_C
  properties {softTaskIcon=&T_UI_NPC_Back}
  sub_nodes [ConversationChoiceNode_2 {class=ConversationChoiceNode, properties {defaultChoiceDisplayText=BACK}}]
 }
]
